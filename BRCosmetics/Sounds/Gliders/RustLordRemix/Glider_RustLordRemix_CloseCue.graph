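# Sound cue for the closing sound of the RustLordRemix glider cosmetic.
# A single wave player feeds the Glider_RustLordRemix_Close01 wave through a modulator that applies a fixed, reduced volume with no random spread.
# The cue uses the Athena glider open/close attenuation and concurrency, the glider land sound class, and the shared cosmetic glider-close audio property sheet.
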SoundCue /BRCosmetics/Sounds/Gliders/RustLordRemix/Glider_RustLordRemix_CloseCue
root=Modulator_0
nodes [
 {id=Modulator_0, children [WavePlayer_1], properties {volumeMax=0.45, volumeMin=0.45}}
 {id=WavePlayer_1, properties {soundWaveAssetPtr=&/Glider_RustLordRemix_Close01}}
]
properties {
 attenuationSettings=/Game/Sounds/Attenuation/Fort_Player_Glider_OpenClose_Athena_Att
 audioPropertiesSheet=/BRCosmetics/Sounds/Settings/PropertySheets/Glider/APS_Cosmetic_Glider_Close
 maxDistance=10000
 soundClassObject=&Fort_Gliders_Land
 volumeMultiplier=0.7
 concurrencySet [&Athena_GliderOpenClose_Conc]
}
