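# Sound cue for the looping audio of the RustLordRemix glider cosmetic.
# A mixer layers two looping waves, Loop01 at slightly reduced level and Loop03 at full level.
# The cue plays at reduced overall volume, with player glider thrust attenuation, a cosmetic glider loop audio properties sheet and the glider landing sound class.
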SoundCue /BRCosmetics/Sounds/Gliders/RustLordRemix/Glider_RustLordRemix_Loop_Cue
root=Mixer_0
nodes [
 {id=Mixer_0, children [WavePlayer_0,WavePlayer_3], properties {inputVolume [0.8,1]}}
 {id=WavePlayer_0, properties {bLooping=true, soundWaveAssetPtr=&/Glider_RustLordRemix_Loop01}}
 {id=WavePlayer_3, properties {bLooping=true, soundWaveAssetPtr=&/Glider_RustLordRemix_Loop03}}
]
properties {
 attenuationSettings=&Fort_Player_Glider_Thrust_Athena_Att
 audioPropertiesSheet=&APS_Cosmetic_Glider_Loop
 maxDistance=12500
 soundClassObject=&Fort_Gliders_Land
 volumeMultiplier=0.32
}
