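# Sound cue for the opening sound of the RustLordRemix glider cosmetic.
# A single wave player feeds a modulator that lowers volume at fixed pitch, then a short fixed delay at the root.
# It uses the shared glider open/close attenuation and concurrency, the cosmetic glider-open audio property sheet, and the Fort_Gliders_Land sound class.
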SoundCue /BRCosmetics/Sounds/Gliders/RustLordRemix/Glider_RustLordRemix_Open_Cue
root=Delay_0
nodes [
 {id=Delay_0, children [Modulator_0], properties {delayMax=0.3, delayMin=0.3}}
 {id=Modulator_0, children [WavePlayer_5], properties {pitchMax=1, pitchMin=1, volumeMax=0.55, volumeMin=0.55}}
 {id=WavePlayer_5, properties {soundWaveAssetPtr=&/Glider_RustLordRemix_Open01}}
]
properties {
 attenuationSettings=/Game/Sounds/Attenuation/Fort_Player_Glider_OpenClose_Athena_Att
 audioPropertiesSheet=/BRCosmetics/Sounds/Settings/PropertySheets/Glider/APS_Cosmetic_Glider_Open
 maxDistance=10000
 soundClassObject=&Fort_Gliders_Land
 volumeMultiplier=1
 concurrencySet [&Athena_GliderOpenClose_Conc]
}
