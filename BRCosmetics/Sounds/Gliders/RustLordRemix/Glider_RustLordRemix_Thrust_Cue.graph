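# Sound cue for the thrust sound of the RustLordRemix glider cosmetic.
# It mixes two random picks at equal levels: one of two shared DieselPunk thrust waves and one of three RustLordRemix-specific thrust waves.
# A modulator at the root varies the mixed result, with its upper pitch capped at normal.
# Playback uses player glider open/close attenuation, a glider sound class and glider thrust concurrency, at reduced volume.
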SoundCue /BRCosmetics/Sounds/Gliders/RustLordRemix/Glider_RustLordRemix_Thrust_Cue
root=Modulator_0
nodes [
 {id=Mixer_2, children [Random_0,Random_1], properties {inputVolume [1,1]}}
 {id=Modulator_0, children [Mixer_2], properties {pitchMax=1}}
 {id=Random_0, children [WavePlayer_3,WavePlayer_4], properties {weights [1,1]}}
 {id=Random_1, children [WavePlayer_5,WavePlayer_6,WavePlayer_7], properties {weights [1,1,1]}}
 {id=WavePlayer_3, properties {soundWaveAssetPtr=/BRCosmetics/Sounds/Gliders/SFX_Glider_DieselPunk_Thrust_03}}
 {id=WavePlayer_4, properties {soundWaveAssetPtr=/BRCosmetics/Sounds/Gliders/SFX_Glider_DieselPunk_Thrust_04}}
 {id=WavePlayer_5, properties {soundWaveAssetPtr=&/Glider_RustLordRemix_Thrust01}}
 {id=WavePlayer_6, properties {soundWaveAssetPtr=&/Glider_RustLordRemix_Thrust02}}
 {id=WavePlayer_7, properties {soundWaveAssetPtr=&/Glider_RustLordRemix_Thrust03}}
]
properties {
 attenuationSettings=/Game/Sounds/Attenuation/Fort_Player_Glider_OpenClose_Athena_Att
 audioPropertiesSheet=&APS_Cosmetic_Glider_Loop
 maxDistance=10000
 soundClassObject=&Fort_Gliders_Land
 volumeMultiplier=0.2
 concurrencySet [&Athena_Glider_Thrust_Conc]
}
